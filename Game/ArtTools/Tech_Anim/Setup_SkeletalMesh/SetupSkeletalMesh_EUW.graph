EditorUtilityWidgetBlueprint /Game/ArtTools/Tech_Anim/Setup_SkeletalMesh/SetupSkeletalMesh_EUW
graphs [EventGraph {nodes [
   "On Clicked (GenerateMeshButton)"() {is=title K2Node_ComponentBoundEvent}
   > Set Valid_StaticMesh(Valid_StaticMesh:bool=true) -> (Output_Get:bool)
   > n1 = Sequence()
   > n2 = Get!(execute:Exec) -> (then:Exec, else:Exec, StaticMesh:Static Mesh Object Reference)
   > Set StaticMesh!(
    execute
    StaticMesh:Static Mesh Object Reference=@n2
    self:Static To Skeletal Mesh BP Object Reference=Get StaticToSkeletal_BP() -> (StaticToSkeletal_BP:Static To Skeletal Mesh BP Object Reference)
   ) -> (then, Output_Get:Static Mesh Object Reference)
   Set Valid_StaticMesh(execute=@n2.else, Valid_StaticMesh=false)
   n3 = Branch(execute=@n1.then_1, Condition=Get Valid_StaticMesh() -> (Valid_StaticMesh:bool))
   > n4 = LoadAndGetAssetObject!(
    execute
    self:EUW Custom Asset Picker Object Reference=Get EUW_CustomAssetPicker() -> (EUW_CustomAssetPicker:EUW Custom Asset Picker Object Reference)
   ) -> (then, asset:Object Reference, HasSelection:bool)
   > n5 = Cast To SkeletalMesh!(execute, Object=@n4) -> (then, CastFailed, AsSkeletal Mesh:Skeletal Mesh Object Reference) {is=title K2Node_DynamicCast}
   n6 = Get StaticToSkeletal_BP()
   > Set Template!(execute, Template:Skeletal Mesh Object Reference=@n5, self:Static To Skeletal Mesh BP Object Reference=@n6) -> (then, Output_Get:Skeletal Mesh Object Reference)
   > Set prune_fingers_skinning!(
    execute
    prune_fingers_skinning:bool=IsChecked(self=Get checkBox_fingers_skinning() -> (checkBox_fingers_skinning:Editor Utility Check Box Object Reference))
    self:Static To Skeletal Mesh BP Object Reference=@n6
   ) -> (then, Output_Get:bool)
   n7 = Get AssetName_TextBox() -> (AssetName_TextBox:Editor Utility Editable Text Box Object Reference)
   > n8 = processStaticToSkeletal!(
    execute
    self:Static To Skeletal Mesh BP Object Reference=@n6
    AssetName:String="To String (Text)"(InText=Get Text(self=@n7))
    CreateAnimBP:bool=true
    CreateSkeleton:bool=true
   ) -> (then, skelmesh:Skeletal Mesh Object Reference)
   > Set SkeletalMesh(SkeletalMesh:Skeletal Mesh Object Reference=@n8) -> (Output_Get:Skeletal Mesh Object Reference)
   > SetObject(self=Get SkeletalMesh_DetailsView() -> (SkeletalMesh_DetailsView:Details View Object Reference), NewObject=Self-Reference())
   Show Message Dialog(
    execute=@n3.else
    Title=MakeLiteralText(Value="LOCTEXT(\"Create Skeletal Mesh: Failed\")")
    Message=MakeLiteralText(Value="LOCTEXT(\"A valid Static Mesh was not assigned.\")")
    MessageType=Ok
    DefaultValue=Ok
    MessageCategory=Error
   )
   "On Property Changed (StaticMesh_DetailsView)"!() -> (OutputDelegate, then, PropertyName:Name=None) {is=title K2Node_ComponentBoundEvent}
   > n9 = Get()
   n10 = GetDisplayName(Object=@n9)
   > Set Text(Text="To Text (String)"(InString=@n10), self:Text Box Object Reference=Get AssetName_TextBox())
   > Set AssetName(AssetName:String=@n10) -> (Output_Get:String)
   "On Clicked (CreateDataAssetsButton)"() {is=title K2Node_ComponentBoundEvent}
   > n11 = Get!(execute) -> (then, else, SkeletalMesh:Skeletal Mesh Object Reference)
   n12 = Get AssetName_TextBox()
   > n13 = Set AssetName(AssetName="To String (Text)"(InText=Get Text(self=@n12)))
   > n14 = processDataAssets!(execute, self:Static To Skeletal Mesh BP Object Reference=Get StaticToSkeletal_BP(), AssetName:String=@n13) -> (then, CharacterItemDef:Athena Character Item Definition Object Reference)
   > Set Character Item Def(Character Item Def:Athena Character Item Definition Object Reference=@n14) -> (Output_Get:Athena Character Item Definition Object Reference)
   > SetObject(self=Get DataAssets_DetailsView() -> (DataAssets_DetailsView:Details View Object Reference), NewObject=Self-Reference())
   Show Message Dialog(
    execute=@n11.else
    Title=MakeLiteralText(Value="LOCTEXT(\"Create Data Assets: Failed\")")
    Message=MakeLiteralText(Value="LOCTEXT(\"The output Skeletal Mesh was not created!\")")
    MessageType=Ok
    DefaultValue=Ok
    MessageCategory=Error
   )
   Event PreConstruct()
   > SetObject(self=Get StaticMesh_DetailsView() -> (StaticMesh_DetailsView:Details View Object Reference), NewObject=Self-Reference())
   > SetObject(self=Get SkeletalMesh_DetailsView(), NewObject=Self-Reference())
   > SetObject(self=Get DataAssets_DetailsView(), NewObject=Self-Reference())
   Event Construct()
   n15 = Get EUW_CustomAssetPicker()
   > Set ExactAssetPackageNames!(
    execute
    ExactAssetPackageNames:Array of Names=Get Templates_Names() -> (Templates_Names:Array of Names)
    self:EUW Custom Asset Picker Object Reference=@n15
   ) -> (then, Output_Get:Array of Names)
   > ValidateOptions(self:EUW Custom Asset Picker Object Reference=[@n15, Get EUW_CustomAssetPicker()])
   n16 = Get Templates_Names()
   > SetSelectedAssetPackageName!(
    execute
    self:EUW Custom Asset Picker Object Reference=Get EUW_CustomAssetPicker()
    PackageName:String="To String (Name)"(InName="Get (a ref)"(Array:Array of Names=@n16, Dimension 1=0) -> (Output:"Name (by ref)"))
   ) -> (then)
   > ForceSetSelection(self:EUW Custom Asset Picker Object Reference=Get EUW_CustomAssetPicker())
   > n17 = Construct Object from Class(Class=&/StaticToSkeletalMesh_BP.StaticToSkeletalMesh_BP_C) -> (ReturnValue:Static To Skeletal Mesh BP Object Reference)
   > Set StaticToSkeletal_BP(StaticToSkeletal_BP:Static To Skeletal Mesh BP Object Reference=@n17) -> (Output_Get:Static To Skeletal Mesh BP Object Reference)
  ]}]
variables [
 AssetName {type=string, default=""}
 Character Item Def {type=AthenaCharacterItemDefinition, default=null}
 SkeletalMesh {type=SkeletalMesh, default=null}
 StaticMesh {type=StaticMesh, default=null}
 StaticToSkeletal_BP {type=StaticToSkeletalMesh_BP_C, default=null}
 Templates_Names {
  type="array<string>"
  default [
   /Game/ArtTools/Tech_Anim/Setup_SkeletalMesh/Templates/M_MED_Template
   /Game/ArtTools/Tech_Anim/Setup_SkeletalMesh/Templates/F_MED_Template
   /Game/ArtTools/Tech_Anim/Setup_SkeletalMesh/Templates/M_LRG_Template
   /Game/ArtTools/Tech_Anim/Setup_SkeletalMesh/Templates/F_LRG_Template
  ]
 }
 Valid_StaticMesh {type=boolean}
]
widgets {
 info {rootWidgetClass=&StackBox, widgetCount=50}
 widgets [
  StackBox_0 {widgetClassPath=&StackBox, widget=StackBox_0}
  VerticalBox_47 {
   widgetClassPath=&VerticalBox
   parent=StackBox_0
   widget=VerticalBox_47
   slot {class=StackBoxSlot, properties {size {sizeRule=Fill, value=1}}}
  }
  HorizontalBox_112 {widgetClassPath=&HorizontalBox, parent=VerticalBox_47, widget=HorizontalBox_112, slot {class=VerticalBoxSlot}}
  Overlay_116 {
   widgetClassPath=&Overlay
   parent=HorizontalBox_112
   widget=Overlay_116
   slot {class=HorizontalBoxSlot, properties {size {sizeRule=Fill, value=1}}}
  }
  MorphHeader {
   widgetClassPath=&Image
   parent=Overlay_116
   widget=MorphHeader
   properties {brush {
     bIsDynamicallyLoaded=false
     drawAs=Image
     imageType=NoImage
     mirroring=NoMirror
     resourceName=None
     resourceObject=null
     tiling=NoTile
     imageSize=<32 32>
     margin=<ltrb 0 0 0 0>
     outlineSettings {
      bUseBrushTransparency=false
      roundingType=HalfHeightRadius
      width=0
      color {colorUseRule=UseColor_Specified, specifiedColor=<rgba 0 0 0 0>}
      cornerRadii=<0 0 0 0>
     }
     tintColor {colorUseRule=UseColor_Specified, specifiedColor=<rgba 0.05 0.05 0.05 1>}
     uVRegion {bIsValid=false, max=<0 0>, min=<0 0>}
    }}
   slot {class=OverlaySlot, properties {horizontalAlignment=HAlign_Fill, verticalAlignment=VAlign_Fill}}
  }
  TextBlock_40 {
   widgetClassPath=&TextBlock
   parent=Overlay_116
   widget=TextBlock_40
   properties {
    text=Static Mesh
    font {
     bForceMonospaced=false
     bMaterialIsStencil=false
     fontMaterial=null
     fontObject=&Roboto
     letterSpacing=0
     monospacedWidth=1
     size=12
     skewAmount=0
     typefaceFontName=Bold
     outlineSettings {
      bApplyOutlineToDropShadows=false
      bMiteredCorners=false
      bSeparateFillAlpha=false
      outlineMaterial=null
      outlineSize=0
      outlineColor=<rgba 0 0 0 1>
     }
    }
   }
   slot {class=OverlaySlot, properties {verticalAlignment=VAlign_Fill, padding=<ltrb 8 8 16 0>}}
  }
  StaticMesh_DetailsView {
   widgetClassPath=&DetailsView
   parent=VerticalBox_47
   widget=StaticMesh_DetailsView
   properties {
    bShowAnimatedPropertiesOption=false
    bShowKeyablePropertiesOption=false
    bShowScrollBar=false
    columnWidth=0.8
    categoriesToShow [StaticMesh]
   }
   slot {class=VerticalBoxSlot}
  }
  HorizontalBox {widgetClassPath=&HorizontalBox, parent=VerticalBox_47, widget=HorizontalBox, slot {class=VerticalBoxSlot}}
  Overlay {
   widgetClassPath=&Overlay
   parent=HorizontalBox
   widget=Overlay
   slot {class=HorizontalBoxSlot, properties {size {sizeRule=Fill, value=1}}}
  }
  MorphHeader_1 {
   widgetClassPath=&Image
   parent=Overlay
   widget=MorphHeader_1
   properties {brush {
     bIsDynamicallyLoaded=false
     drawAs=Image
     imageType=NoImage
     mirroring=NoMirror
     resourceName=None
     resourceObject=null
     tiling=NoTile
     imageSize=<32 32>
     margin=<ltrb 0 0 0 0>
     outlineSettings {
      bUseBrushTransparency=false
      roundingType=HalfHeightRadius
      width=0
      color {colorUseRule=UseColor_Specified, specifiedColor=<rgba 0 0 0 0>}
      cornerRadii=<0 0 0 0>
     }
     tintColor {colorUseRule=UseColor_Specified, specifiedColor=<rgba 0.05 0.05 0.05 1>}
     uVRegion {bIsValid=false, max=<0 0>, min=<0 0>}
    }}
   slot {class=OverlaySlot, properties {horizontalAlignment=HAlign_Fill, verticalAlignment=VAlign_Fill}}
  }
  TextBlock {
   widgetClassPath=&TextBlock
   parent=Overlay
   widget=TextBlock
   properties {
    text=Templates
    font {
     bForceMonospaced=false
     bMaterialIsStencil=false
     fontMaterial=null
     fontObject=&Roboto
     letterSpacing=0
     monospacedWidth=1
     size=12
     skewAmount=0
     typefaceFontName=Bold
     outlineSettings {
      bApplyOutlineToDropShadows=false
      bMiteredCorners=false
      bSeparateFillAlpha=false
      outlineMaterial=null
      outlineSize=0
      outlineColor=<rgba 0 0 0 1>
     }
    }
   }
   slot {class=OverlaySlot, properties {verticalAlignment=VAlign_Fill, padding=<ltrb 8 8 16 0>}}
  }
  VerticalBox_40 {widgetClassPath=&VerticalBox, parent=VerticalBox_47, widget=VerticalBox_40, slot {class=VerticalBoxSlot}}
  HorizontalBox_59 {
   widgetClassPath=&HorizontalBox
   parent=VerticalBox_40
   widget=HorizontalBox_59
   slot {class=VerticalBoxSlot, properties {padding=<ltrb 0 10 0 15>}}
  }
  EUW_CustomAssetPicker {
   widgetClassPath=/Game/NeverCook/EditorUtilities/CustomAssetPicker/EUW_CustomAssetPicker
   bIsVariable=true
   parent=HorizontalBox_59
   widget=EUW_CustomAssetPicker
   properties {label="     Template Mesh:", labelTextSize=8}
   slot {class=HorizontalBoxSlot, properties {size {sizeRule=Fill, value=1}}}
  }
  HorizontalBox_1 {widgetClassPath=&HorizontalBox, parent=VerticalBox_47, widget=HorizontalBox_1, slot {class=VerticalBoxSlot}}
  Overlay_1 {
   widgetClassPath=&Overlay
   parent=HorizontalBox_1
   widget=Overlay_1
   slot {class=HorizontalBoxSlot, properties {size {sizeRule=Fill, value=1}}}
  }
  MorphHeader_2 {
   widgetClassPath=&Image
   parent=Overlay_1
   widget=MorphHeader_2
   properties {brush {
     bIsDynamicallyLoaded=false
     drawAs=Image
     imageType=NoImage
     mirroring=NoMirror
     resourceName=None
     resourceObject=null
     tiling=NoTile
     imageSize=<32 32>
     margin=<ltrb 0 0 0 0>
     outlineSettings {
      bUseBrushTransparency=false
      roundingType=HalfHeightRadius
      width=0
      color {colorUseRule=UseColor_Specified, specifiedColor=<rgba 0 0 0 0>}
      cornerRadii=<0 0 0 0>
     }
     tintColor {colorUseRule=UseColor_Specified, specifiedColor=<rgba 0.05 0.05 0.05 1>}
     uVRegion {bIsValid=false, max=<0 0>, min=<0 0>}
    }}
   slot {class=OverlaySlot, properties {horizontalAlignment=HAlign_Fill, verticalAlignment=VAlign_Fill}}
  }
  TextBlock_1 {
   widgetClassPath=&TextBlock
   parent=Overlay_1
   widget=TextBlock_1
   properties {
    text=Asset Name
    font {
     bForceMonospaced=false
     bMaterialIsStencil=false
     fontMaterial=null
     fontObject=&Roboto
     letterSpacing=0
     monospacedWidth=1
     size=12
     skewAmount=0
     typefaceFontName=Bold
     outlineSettings {
      bApplyOutlineToDropShadows=false
      bMiteredCorners=false
      bSeparateFillAlpha=false
      outlineMaterial=null
      outlineSize=0
      outlineColor=<rgba 0 0 0 1>
     }
    }
   }
   slot {class=OverlaySlot, properties {verticalAlignment=VAlign_Fill, padding=<ltrb 8 8 16 0>}}
  }
  VerticalBox {
   widgetClassPath=&VerticalBox
   parent=VerticalBox_47
   widget=VerticalBox
   slot {class=VerticalBoxSlot, properties {padding=<ltrb 0 5 0 0>}}
  }
  HorizontalBox_2 {
   widgetClassPath=&HorizontalBox
   parent=VerticalBox
   widget=HorizontalBox_2
   slot {class=VerticalBoxSlot, properties {padding=<ltrb 20 0 0 10>}}
  }
  CommonTextBlock_1 {
   widgetClassPath=&CommonTextBlock
   parent=HorizontalBox_2
   widget=CommonTextBlock_1
   properties {
    text=Asset Name:
    toolTipText="If True: Changing the character will also attempt to change the other cosmetics to match.  \r\n\r\nDisable to set each individually."
    font {
     bForceMonospaced=false
     bMaterialIsStencil=false
     fontMaterial=null
     fontObject=&NotoSans
     letterSpacing=0
     monospacedWidth=1
     size=7.5
     skewAmount=0
     typefaceFontName=NotoSans-Regular
     outlineSettings {
      bApplyOutlineToDropShadows=false
      bMiteredCorners=false
      bSeparateFillAlpha=false
      outlineMaterial=null
      outlineSize=0
      outlineColor=<rgba 0 0 0 1>
     }
    }
    shadowOffset=<0 0>
   }
   slot {class=HorizontalBoxSlot, properties {horizontalAlignment=HAlign_Center, verticalAlignment=VAlign_Center}}
  }
  Spacer_338 {widgetClassPath=&Spacer, parent=HorizontalBox_2, widget=Spacer_338, slot {class=HorizontalBoxSlot}}
  HorizontalBox_9 {
   widgetClassPath=&HorizontalBox
   parent=HorizontalBox_2
   widget=HorizontalBox_9
   slot {class=HorizontalBoxSlot, properties {padding=<ltrb 25 10 0 10>, size {sizeRule=Fill, value=1}}}
  }
  AssetName_TextBox {
   widgetClassPath=&EditorUtilityEditableTextBox
   bIsVariable=true
   parent=HorizontalBox_9
   widget=AssetName_TextBox
   properties {
    hintText=Enter Asset Name
    widgetStyle {
     backgroundColor {colorUseRule=UseColor_Specified, specifiedColor=<rgba 1 1 1 1>}
     backgroundImageFocused {
      bIsDynamicallyLoaded=false
      drawAs=RoundedBox
      imageType=NoImage
      mirroring=NoMirror
      resourceName=None
      resourceObject=null
      tiling=NoTile
      imageSize=<0 0>
      margin=<ltrb 0 0 0 0>
      outlineSettings {
       bUseBrushTransparency=false
       roundingType=FixedRadius
       width=1
       color {colorUseRule=UseColor_ColorTable, specifiedColor=<rgba 0 0 0 0>}
       cornerRadii=<4 4 4 4>
      }
      tintColor {colorUseRule=UseColor_Specified, specifiedColor=<rgba 0.0048 0.0048 0.0048 1>}
      uVRegion {bIsValid=false, max=<0 0>, min=<0 0>}
     }
     backgroundImageHovered {
      bIsDynamicallyLoaded=false
      drawAs=RoundedBox
      imageType=NoImage
      mirroring=NoMirror
      resourceName=None
      resourceObject=null
      tiling=NoTile
      imageSize=<0 0>
      margin=<ltrb 0 0 0 0>
      outlineSettings {
       bUseBrushTransparency=false
       roundingType=FixedRadius
       width=1
       color {colorUseRule=UseColor_ColorTable, specifiedColor=<rgba 0 0 0 0>}
       cornerRadii=<4 4 4 4>
      }
      tintColor {colorUseRule=UseColor_Specified, specifiedColor=<rgba 0.0048 0.0048 0.0048 1>}
      uVRegion {bIsValid=false, max=<0 0>, min=<0 0>}
     }
     backgroundImageNormal {
      bIsDynamicallyLoaded=false
      drawAs=RoundedBox
      imageType=NoImage
      mirroring=NoMirror
      resourceName=None
      resourceObject=null
      tiling=NoTile
      imageSize=<0 0>
      margin=<ltrb 0 0 0 0>
      outlineSettings {
       bUseBrushTransparency=false
       roundingType=FixedRadius
       width=1
       color {colorUseRule=UseColor_ColorTable, specifiedColor=<rgba 0 0 0 0>}
       cornerRadii=<4 4 4 4>
      }
      tintColor {colorUseRule=UseColor_Specified, specifiedColor=<rgba 0.0048 0.0048 0.0048 1>}
      uVRegion {bIsValid=false, max=<0 0>, min=<0 0>}
     }
     backgroundImageReadOnly {
      bIsDynamicallyLoaded=false
      drawAs=RoundedBox
      imageType=NoImage
      mirroring=NoMirror
      resourceName=None
      resourceObject=null
      tiling=NoTile
      imageSize=<0 0>
      margin=<ltrb 0 0 0 0>
      outlineSettings {
       bUseBrushTransparency=false
       roundingType=FixedRadius
       width=1
       color {colorUseRule=UseColor_ColorTable, specifiedColor=<rgba 0 0 0 0>}
       cornerRadii=<4 4 4 4>
      }
      tintColor {colorUseRule=UseColor_Specified, specifiedColor=<rgba 0.0284 0.0284 0.0284 1>}
      uVRegion {bIsValid=false, max=<0 0>, min=<0 0>}
     }
     focusedForegroundColor {colorUseRule=UseColor_Specified, specifiedColor=<rgba 1 1 1 1>}
     foregroundColor {colorUseRule=UseColor_Specified, specifiedColor=<rgba 0.5271 0.5271 0.5271 1>}
     hScrollBarPadding=<ltrb 0 0 0 0>
     padding=<ltrb 8 6 8 6>
     readOnlyForegroundColor {colorUseRule=UseColor_Foreground, specifiedColor=<rgba 1 0 1 1>}
     scrollBarStyle {
      thickness=8
      draggedThumbImage {
       bIsDynamicallyLoaded=false
       drawAs=RoundedBox
       imageType=NoImage
       mirroring=NoMirror
       resourceName=None
       resourceObject=null
       tiling=NoTile
       imageSize=<0 0>
       margin=<ltrb 0 0 0 0>
       outlineSettings {
        bUseBrushTransparency=false
        roundingType=FixedRadius
        width=0
        color {colorUseRule=UseColor_Specified, specifiedColor=<rgba 0 0 0 0>}
        cornerRadii=<4 4 4 4>
       }
       tintColor {colorUseRule=UseColor_Specified, specifiedColor=<rgba 0.2159 0.2159 0.2159 1>}
       uVRegion {bIsValid=false, max=<0 0>, min=<0 0>}
      }
      horizontalBackgroundImage {
       bIsDynamicallyLoaded=false
       drawAs=NoDrawType
       imageType=NoImage
       mirroring=NoMirror
       resourceName=None
       resourceObject=null
       tiling=NoTile
       imageSize=<0 0>
       margin=<ltrb 0 0 0 0>
       outlineSettings {
        bUseBrushTransparency=false
        roundingType=HalfHeightRadius
        width=0
        color {colorUseRule=UseColor_Specified, specifiedColor=<rgba 0 0 0 0>}
        cornerRadii=<0 0 0 0>
       }
       tintColor {colorUseRule=UseColor_Specified, specifiedColor=<rgba 1 1 1 1>}
       uVRegion {bIsValid=false, max=<0 0>, min=<0 0>}
      }
      horizontalBottomSlotImage {
       bIsDynamicallyLoaded=false
       drawAs=NoDrawType
       imageType=NoImage
       mirroring=NoMirror
       resourceName=None
       resourceObject=null
       tiling=NoTile
       imageSize=<0 0>
       margin=<ltrb 0 0 0 0>
       outlineSettings {
        bUseBrushTransparency=false
        roundingType=HalfHeightRadius
        width=0
        color {colorUseRule=UseColor_Specified, specifiedColor=<rgba 0 0 0 0>}
        cornerRadii=<0 0 0 0>
       }
       tintColor {colorUseRule=UseColor_Specified, specifiedColor=<rgba 1 1 1 1>}
       uVRegion {bIsValid=false, max=<0 0>, min=<0 0>}
      }
      horizontalTopSlotImage {
       bIsDynamicallyLoaded=false
       drawAs=NoDrawType
       imageType=NoImage
       mirroring=NoMirror
       resourceName=None
       resourceObject=null
       tiling=NoTile
       imageSize=<0 0>
       margin=<ltrb 0 0 0 0>
       outlineSettings {
        bUseBrushTransparency=false
        roundingType=HalfHeightRadius
        width=0
        color {colorUseRule=UseColor_Specified, specifiedColor=<rgba 0 0 0 0>}
        cornerRadii=<0 0 0 0>
       }
       tintColor {colorUseRule=UseColor_Specified, specifiedColor=<rgba 1 1 1 1>}
       uVRegion {bIsValid=false, max=<0 0>, min=<0 0>}
      }
      hoveredThumbImage {
       bIsDynamicallyLoaded=false
       drawAs=RoundedBox
       imageType=NoImage
       mirroring=NoMirror
       resourceName=None
       resourceObject=null
       tiling=NoTile
       imageSize=<0 0>
       margin=<ltrb 0 0 0 0>
       outlineSettings {
        bUseBrushTransparency=false
        roundingType=FixedRadius
        width=0
        color {colorUseRule=UseColor_Specified, specifiedColor=<rgba 0 0 0 0>}
        cornerRadii=<4 4 4 4>
       }
       tintColor {colorUseRule=UseColor_Specified, specifiedColor=<rgba 0.2159 0.2159 0.2159 1>}
       uVRegion {bIsValid=false, max=<0 0>, min=<0 0>}
      }
      normalThumbImage {
       bIsDynamicallyLoaded=false
       drawAs=RoundedBox
       imageType=NoImage
       mirroring=NoMirror
       resourceName=None
       resourceObject=null
       tiling=NoTile
       imageSize=<0 0>
       margin=<ltrb 0 0 0 0>
       outlineSettings {
        bUseBrushTransparency=false
        roundingType=FixedRadius
        width=0
        color {colorUseRule=UseColor_Specified, specifiedColor=<rgba 0 0 0 0>}
        cornerRadii=<4 4 4 4>
       }
       tintColor {colorUseRule=UseColor_Specified, specifiedColor=<rgba 0.0953 0.0953 0.0953 1>}
       uVRegion {bIsValid=false, max=<0 0>, min=<0 0>}
      }
      verticalBackgroundImage {
       bIsDynamicallyLoaded=false
       drawAs=NoDrawType
       imageType=NoImage
       mirroring=NoMirror
       resourceName=None
       resourceObject=null
       tiling=NoTile
       imageSize=<0 0>
       margin=<ltrb 0 0 0 0>
       outlineSettings {
        bUseBrushTransparency=false
        roundingType=HalfHeightRadius
        width=0
        color {colorUseRule=UseColor_Specified, specifiedColor=<rgba 0 0 0 0>}
        cornerRadii=<0 0 0 0>
       }
       tintColor {colorUseRule=UseColor_Specified, specifiedColor=<rgba 1 1 1 1>}
       uVRegion {bIsValid=false, max=<0 0>, min=<0 0>}
      }
      verticalBottomSlotImage {
       bIsDynamicallyLoaded=false
       drawAs=NoDrawType
       imageType=NoImage
       mirroring=NoMirror
       resourceName=None
       resourceObject=null
       tiling=NoTile
       imageSize=<0 0>
       margin=<ltrb 0 0 0 0>
       outlineSettings {
        bUseBrushTransparency=false
        roundingType=HalfHeightRadius
        width=0
        color {colorUseRule=UseColor_Specified, specifiedColor=<rgba 0 0 0 0>}
        cornerRadii=<0 0 0 0>
       }
       tintColor {colorUseRule=UseColor_Specified, specifiedColor=<rgba 1 1 1 1>}
       uVRegion {bIsValid=false, max=<0 0>, min=<0 0>}
      }
      verticalTopSlotImage {
       bIsDynamicallyLoaded=false
       drawAs=NoDrawType
       imageType=NoImage
       mirroring=NoMirror
       resourceName=None
       resourceObject=null
       tiling=NoTile
       imageSize=<0 0>
       margin=<ltrb 0 0 0 0>
       outlineSettings {
        bUseBrushTransparency=false
        roundingType=HalfHeightRadius
        width=0
        color {colorUseRule=UseColor_Specified, specifiedColor=<rgba 0 0 0 0>}
        cornerRadii=<0 0 0 0>
       }
       tintColor {colorUseRule=UseColor_Specified, specifiedColor=<rgba 1 1 1 1>}
       uVRegion {bIsValid=false, max=<0 0>, min=<0 0>}
      }
     }
     textStyle {
      overflowPolicy=Clip
      transformPolicy=None
      colorAndOpacity {colorUseRule=UseColor_Foreground, specifiedColor=<rgba 1 0 1 1>}
      font {
       bForceMonospaced=false
       bMaterialIsStencil=false
       fontMaterial=null
       fontObject=null
       letterSpacing=0
       monospacedWidth=1
       size=10
       skewAmount=0
       typefaceFontName=Regular
       outlineSettings {
        bApplyOutlineToDropShadows=false
        bMiteredCorners=false
        bSeparateFillAlpha=false
        outlineMaterial=null
        outlineSize=0
        outlineColor=<rgba 0 0 0 1>
       }
      }
      highlightColor {colorUseRule=UseColor_ColorTable, specifiedColor=<rgba 0 0 0 0>}
      highlightShape {
       bIsDynamicallyLoaded=false
       drawAs=Image
       imageType=NoImage
       mirroring=NoMirror
       resourceName=None
       resourceObject=null
       tiling=NoTile
       imageSize=<0 0>
       margin=<ltrb 0 0 0 0>
       outlineSettings {
        bUseBrushTransparency=false
        roundingType=HalfHeightRadius
        width=0
        color {colorUseRule=UseColor_Specified, specifiedColor=<rgba 0 0 0 0>}
        cornerRadii=<0 0 0 0>
       }
       tintColor {colorUseRule=UseColor_ColorTable, specifiedColor=<rgba 0 0 0 0>}
       uVRegion {bIsValid=false, max=<0 0>, min=<0 0>}
      }
      selectedBackgroundColor {colorUseRule=UseColor_ColorTable, specifiedColor=<rgba 0 0 0 0>}
      shadowColorAndOpacity=<rgba 0 0 0 1>
      shadowOffset=<0 0>
      strikeBrush {
       bIsDynamicallyLoaded=false
       drawAs=Image
       imageType=NoImage
       mirroring=NoMirror
       resourceName=None
       resourceObject=null
       tiling=NoTile
       imageSize=<32 32>
       margin=<ltrb 0 0 0 0>
       outlineSettings {
        bUseBrushTransparency=false
        roundingType=HalfHeightRadius
        width=0
        color {colorUseRule=UseColor_Specified, specifiedColor=<rgba 0 0 0 0>}
        cornerRadii=<0 0 0 0>
       }
       tintColor {colorUseRule=UseColor_Specified, specifiedColor=<rgba 1 1 1 1>}
       uVRegion {bIsValid=false, max=<0 0>, min=<0 0>}
      }
      underlineBrush {
       bIsDynamicallyLoaded=false
       drawAs=Image
       imageType=NoImage
       mirroring=NoMirror
       resourceName=None
       resourceObject=null
       tiling=NoTile
       imageSize=<32 32>
       margin=<ltrb 0 0 0 0>
       outlineSettings {
        bUseBrushTransparency=false
        roundingType=HalfHeightRadius
        width=0
        color {colorUseRule=UseColor_Specified, specifiedColor=<rgba 0 0 0 0>}
        cornerRadii=<0 0 0 0>
       }
       tintColor {colorUseRule=UseColor_Specified, specifiedColor=<rgba 1 1 1 1>}
       uVRegion {bIsValid=false, max=<0 0>, min=<0 0>}
      }
     }
     vScrollBarPadding=<ltrb 0 0 0 0>
    }
   }
   slot {class=HorizontalBoxSlot, properties {verticalAlignment=VAlign_Center, size {sizeRule=Fill, value=1}}}
  }
  Spacer_1 {
   widgetClassPath=&Spacer
   parent=HorizontalBox_9
   widget=Spacer_1
   slot {class=HorizontalBoxSlot, properties {size {sizeRule=Fill, value=0.15}}}
  }
  HorizontalBox_7 {widgetClassPath=&HorizontalBox, parent=VerticalBox_47, widget=HorizontalBox_7, slot {class=VerticalBoxSlot}}
  Overlay_4 {
   widgetClassPath=&Overlay
   parent=HorizontalBox_7
   widget=Overlay_4
   slot {class=HorizontalBoxSlot, properties {size {sizeRule=Fill, value=1}}}
  }
  MorphHeader_5 {
   widgetClassPath=&Image
   parent=Overlay_4
   widget=MorphHeader_5
   properties {brush {
     bIsDynamicallyLoaded=false
     drawAs=Image
     imageType=NoImage
     mirroring=NoMirror
     resourceName=None
     resourceObject=null
     tiling=NoTile
     imageSize=<32 32>
     margin=<ltrb 0 0 0 0>
     outlineSettings {
      bUseBrushTransparency=false
      roundingType=HalfHeightRadius
      width=0
      color {colorUseRule=UseColor_Specified, specifiedColor=<rgba 0 0 0 0>}
      cornerRadii=<0 0 0 0>
     }
     tintColor {colorUseRule=UseColor_Specified, specifiedColor=<rgba 0.05 0.05 0.05 1>}
     uVRegion {bIsValid=false, max=<0 0>, min=<0 0>}
    }}
   slot {class=OverlaySlot, properties {horizontalAlignment=HAlign_Fill, verticalAlignment=VAlign_Fill}}
  }
  TextBlock_4 {
   widgetClassPath=&TextBlock
   parent=Overlay_4
   widget=TextBlock_4
   properties {
    text=Skinning
    font {
     bForceMonospaced=false
     bMaterialIsStencil=false
     fontMaterial=null
     fontObject=&Roboto
     letterSpacing=0
     monospacedWidth=1
     size=12
     skewAmount=0
     typefaceFontName=Bold
     outlineSettings {
      bApplyOutlineToDropShadows=false
      bMiteredCorners=false
      bSeparateFillAlpha=false
      outlineMaterial=null
      outlineSize=0
      outlineColor=<rgba 0 0 0 1>
     }
    }
   }
   slot {class=OverlaySlot, properties {verticalAlignment=VAlign_Fill, padding=<ltrb 8 8 16 0>}}
  }
  VerticalBox_3 {widgetClassPath=&VerticalBox, parent=VerticalBox_47, widget=VerticalBox_3, slot {class=VerticalBoxSlot}}
  HorizontalBox_8 {
   widgetClassPath=&HorizontalBox
   parent=VerticalBox_3
   widget=HorizontalBox_8
   slot {class=VerticalBoxSlot, properties {padding=<ltrb 0 0 0 10>}}
  }
  checkBox_fingers_skinning {
   widgetClassPath=&EditorUtilityCheckBox
   bIsVariable=true
   parent=HorizontalBox_8
   widget=checkBox_fingers_skinning
   properties {checkedState=Checked, toolTipText=Finger bones will be pruned from the skinning. All weights will then be applied to the Hand bones}
   slot {class=HorizontalBoxSlot, properties {padding=<ltrb 20 10 0 10>}}
  }
  CommonTextBlock_2 {
   widgetClassPath=&CommonTextBlock
   parent=checkBox_fingers_skinning
   widget=CommonTextBlock_2
   properties {
    text=Prune Fingers Skinning
    toolTipText="If True: Changing the character will also attempt to change the other cosmetics to match.  \r\n\r\nDisable to set each individually."
    font {
     bForceMonospaced=false
     bMaterialIsStencil=false
     fontMaterial=null
     fontObject=&NotoSans
     letterSpacing=0
     monospacedWidth=1
     size=7.5
     skewAmount=0
     typefaceFontName=NotoSans-Regular
     outlineSettings {
      bApplyOutlineToDropShadows=false
      bMiteredCorners=false
      bSeparateFillAlpha=false
      outlineMaterial=null
      outlineSize=0
      outlineColor=<rgba 0 0 0 1>
     }
    }
    shadowOffset=<0 0>
   }
   slot {class=PanelSlot}
  }
  HorizontalBox_3 {widgetClassPath=&HorizontalBox, parent=VerticalBox_47, widget=HorizontalBox_3, slot {class=VerticalBoxSlot}}
  Overlay_2 {
   widgetClassPath=&Overlay
   parent=HorizontalBox_3
   widget=Overlay_2
   slot {class=HorizontalBoxSlot, properties {size {sizeRule=Fill, value=1}}}
  }
  MorphHeader_3 {
   widgetClassPath=&Image
   parent=Overlay_2
   widget=MorphHeader_3
   properties {brush {
     bIsDynamicallyLoaded=false
     drawAs=Image
     imageType=NoImage
     mirroring=NoMirror
     resourceName=None
     resourceObject=null
     tiling=NoTile
     imageSize=<32 32>
     margin=<ltrb 0 0 0 0>
     outlineSettings {
      bUseBrushTransparency=false
      roundingType=HalfHeightRadius
      width=0
      color {colorUseRule=UseColor_Specified, specifiedColor=<rgba 0 0 0 0>}
      cornerRadii=<0 0 0 0>
     }
     tintColor {colorUseRule=UseColor_Specified, specifiedColor=<rgba 0.05 0.05 0.05 1>}
     uVRegion {bIsValid=false, max=<0 0>, min=<0 0>}
    }}
   slot {class=OverlaySlot, properties {horizontalAlignment=HAlign_Fill, verticalAlignment=VAlign_Fill}}
  }
  TextBlock_2 {
   widgetClassPath=&TextBlock
   parent=Overlay_2
   widget=TextBlock_2
   properties {
    text=Process
    font {
     bForceMonospaced=false
     bMaterialIsStencil=false
     fontMaterial=null
     fontObject=&Roboto
     letterSpacing=0
     monospacedWidth=1
     size=12
     skewAmount=0
     typefaceFontName=Bold
     outlineSettings {
      bApplyOutlineToDropShadows=false
      bMiteredCorners=false
      bSeparateFillAlpha=false
      outlineMaterial=null
      outlineSize=0
      outlineColor=<rgba 0 0 0 1>
     }
    }
   }
   slot {class=OverlaySlot, properties {verticalAlignment=VAlign_Fill, padding=<ltrb 8 8 16 0>}}
  }
  VerticalBox_1 {widgetClassPath=&VerticalBox, parent=VerticalBox_47, widget=VerticalBox_1, slot {class=VerticalBoxSlot}}
  HorizontalBox_4 {
   widgetClassPath=&HorizontalBox
   parent=VerticalBox_1
   widget=HorizontalBox_4
   slot {class=VerticalBoxSlot, properties {padding=<ltrb -50 0 0 0>}}
  }
  GenerateMeshButton {
   widgetClassPath=&EditorUtilityButton
   parent=HorizontalBox_4
   widget=GenerateMeshButton
   properties {renderTransform {angle=0, scale=<0.6 0.6>, shear=<0 0>, translation=<0 0>}}
   slot {class=HorizontalBoxSlot, properties {horizontalAlignment=HAlign_Center, verticalAlignment=VAlign_Center}}
  }
  CommonTextBlock_53 {
   widgetClassPath=&CommonTextBlock
   parent=GenerateMeshButton
   widget=CommonTextBlock_53
   properties {
    style=&TextStyle-Base_C
    text=Create Skeletal Mesh
    font {
     bForceMonospaced=false
     bMaterialIsStencil=false
     fontMaterial=null
     fontObject=&NotoSans
     letterSpacing=0
     monospacedWidth=1
     size=24
     skewAmount=0
     typefaceFontName=NotoSans-Regular
     outlineSettings {
      bApplyOutlineToDropShadows=false
      bMiteredCorners=false
      bSeparateFillAlpha=false
      outlineMaterial=null
      outlineSize=0
      outlineColor=<rgba 0 0 0 1>
     }
    }
    shadowOffset=<0 0>
   }
   slot {class=ButtonSlot, properties {padding=<ltrb 0 0 0 0>}}
  }
  Spacer {widgetClassPath=&Spacer, parent=HorizontalBox_4, widget=Spacer, slot {class=HorizontalBoxSlot}}
  SkeletalMesh_DetailsView {
   widgetClassPath=&DetailsView
   parent=VerticalBox_1
   widget=SkeletalMesh_DetailsView
   properties {
    bShowAnimatedPropertiesOption=false
    bShowKeyablePropertiesOption=false
    bShowScrollBar=false
    columnWidth=0.75
    categoriesToShow [SkeletalMesh]
   }
   slot {class=VerticalBoxSlot, properties {size {sizeRule=Fill, value=1}}}
  }
  VerticalBox_4 {widgetClassPath=&VerticalBox, parent=VerticalBox_47, widget=VerticalBox_4, slot {class=VerticalBoxSlot}}
  VerticalBox_2 {
   widgetClassPath=&VerticalBox
   parent=VerticalBox_47
   widget=VerticalBox_2
   slot {class=VerticalBoxSlot, properties {padding=<ltrb 0 10 0 0>}}
  }
  HorizontalBox_6 {
   widgetClassPath=&HorizontalBox
   parent=VerticalBox_2
   widget=HorizontalBox_6
   slot {class=VerticalBoxSlot, properties {padding=<ltrb -50 0 0 0>}}
  }
  CreateDataAssetsButton {
   widgetClassPath=&EditorUtilityButton
   parent=HorizontalBox_6
   widget=CreateDataAssetsButton
   properties {renderTransform {angle=0, scale=<0.6 0.6>, shear=<0 0>, translation=<0 0>}}
   slot {class=HorizontalBoxSlot, properties {horizontalAlignment=HAlign_Center, verticalAlignment=VAlign_Center}}
  }
  CommonTextBlock {
   widgetClassPath=&CommonTextBlock
   parent=CreateDataAssetsButton
   widget=CommonTextBlock
   properties {
    style=&TextStyle-Base_C
    text="Create Data Assets    "
    font {
     bForceMonospaced=false
     bMaterialIsStencil=false
     fontMaterial=null
     fontObject=&NotoSans
     letterSpacing=0
     monospacedWidth=1
     size=24
     skewAmount=0
     typefaceFontName=NotoSans-Regular
     outlineSettings {
      bApplyOutlineToDropShadows=false
      bMiteredCorners=false
      bSeparateFillAlpha=false
      outlineMaterial=null
      outlineSize=0
      outlineColor=<rgba 0 0 0 1>
     }
    }
    shadowOffset=<0 0>
   }
   slot {class=ButtonSlot, properties {padding=<ltrb 0 0 0 0>}}
  }
  Spacer_2 {widgetClassPath=&Spacer, parent=HorizontalBox_6, widget=Spacer_2, slot {class=HorizontalBoxSlot}}
  DataAssets_DetailsView {
   widgetClassPath=&DetailsView
   parent=VerticalBox_2
   widget=DataAssets_DetailsView
   properties {
    bShowAnimatedPropertiesOption=false
    bShowKeyablePropertiesOption=false
    bShowScrollBar=false
    columnWidth=0.75
    categoriesToShow [ItemDefinition]
   }
   slot {class=VerticalBoxSlot, properties {size {sizeRule=Fill, value=1}}}
  }
 ]
}
