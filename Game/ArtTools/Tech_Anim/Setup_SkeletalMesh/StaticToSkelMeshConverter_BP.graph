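# Editor utility that converts a static mesh into a skeletal mesh using a template skeletal mesh loaded from a path.
# ConvertStaticToSkeletalMesh creates a StaticToSkeletalMesh_BP helper, passes it mesh and template, and requests a new skeleton without an AnimBP.
# The asset name comes from splitting the mesh name on "Temp_" and "_SM".
# It then optionally deletes the source mesh, regenerates the result's LODs and returns the skeletal mesh.
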EditorUtilityBlueprint /Game/ArtTools/Tech_Anim/Setup_SkeletalMesh/StaticToSkelMeshConverter_BP
graphs [
 ConvertStaticToSkeletalMesh {nodes [
   ConvertStaticToSkeletalMesh!() -> (then, StaticMesh:Static Mesh Object Reference, Template SkelMesh Path:String, Delete Static Mesh:bool=true) {is=title K2Node_FunctionEntry}
   > n1 = Sequence()
   > n2 = Construct Object from Class(Class=&/StaticToSkeletalMesh_BP.StaticToSkeletalMesh_BP_C) -> (ReturnValue:Static To Skeletal Mesh BP Object Reference)
   > Set StaticToSkeletalMesh BP(StaticToSkeletalMesh BP:Static To Skeletal Mesh BP Object Reference=@n2) -> (Output_Get:Static To Skeletal Mesh BP Object Reference)
   > n3 = Is Valid(InputObject=Get StaticMesh!() -> (StaticMesh))
   Set Valid_StaticMesh(execute=@n3.Is Not Valid, Valid_StaticMesh:bool=false) -> (Output_Get:bool)
   n4 = Sequence!(execute=@n3.Is Valid) -> (then_0, then_1, then_2:Exec)
   Branch(execute=@n4.then_1, Condition=Get Delete Static Mesh() -> (Delete Static Mesh:bool))
   > n5 = GetPathNameForLoadedAsset(LoadedAsset=Get StaticMesh())
   > DeleteAsset!(execute, AssetPathToDelete:String=@n5) -> (then, ReturnValue:bool)
   n6 = Get StaticMesh()
   Set StaticMesh!(
    execute=@n4.then_0
    StaticMesh:Static Mesh Object Reference=@n6
    self:Static To Skeletal Mesh BP Object Reference=Get StaticToSkeletalMesh BP() -> (StaticToSkeletalMesh BP:Static To Skeletal Mesh BP Object Reference)
   ) -> (then, Output_Get:Static Mesh Object Reference)
   n7 = GetDisplayName(Object=@n6)
   > Set AssetName(AssetName:String=@n7) -> (Output_Get:String)
   > n8 = LoadAsset(AssetPath=Get Template SkelMesh Path() -> (Template SkelMesh Path:String))
   > n9 = Cast To SkeletalMesh!(execute, Object=@n8) -> (then, CastFailed, AsSkeletal Mesh:Skeletal Mesh Object Reference) {is=title K2Node_DynamicCast}
   > Set Template!(execute, Template:Skeletal Mesh Object Reference=@n9, self:Static To Skeletal Mesh BP Object Reference=Get StaticToSkeletalMesh BP()) -> (then, Output_Get:Skeletal Mesh Object Reference)
   > n10 = processStaticToSkeletal!(
    execute
    self:Static To Skeletal Mesh BP Object Reference=Get StaticToSkeletalMesh BP()
    AssetName:String=Split(SourceString=Split(SourceString=@n7, InStr=Temp_, SearchDir=FromStart).RightS, InStr=_SM, SearchDir=FromStart)
    CreateAnimBP:bool=false
    CreateSkeleton:bool=true
   ) -> (then, skelmesh:Skeletal Mesh Object Reference)
   > Set Converted SkelMesh(Converted SkelMesh:Skeletal Mesh Object Reference=@n10) -> (Output_Get:Skeletal Mesh Object Reference)
   RegenerateLOD!(
    execute=@n4.then_2
    SkeletalMesh:Skeletal Mesh Object Reference=Get Converted SkelMesh() -> (Converted SkelMesh:Skeletal Mesh Object Reference)
    NewLODCount:int=4
    bRegenerateEvenIfImported:bool=false
    bGenerateBaseLOD:bool=false
   ) -> (then, ReturnValue:bool)
   Return Node!(execute=@n1.then_1, Skeletal Mesh:Skeletal Mesh Object Reference=Get Converted SkelMesh())
  ]}
 EventGraph {links [], nodes []}
]
variables [
 AssetName {type=string, default=""}
 Converted SkelMesh {type=SkeletalMesh, default=null}
 NewVar {type=StaticToSkelMeshConverter_BP_C, default=null}
 Skelmesh {type=SkeletalMesh, default=null}
 StaticToSkeletalMesh BP {type=StaticToSkeletalMesh_BP_C, default=null}
 Valid_StaticMesh {type=boolean}
]
