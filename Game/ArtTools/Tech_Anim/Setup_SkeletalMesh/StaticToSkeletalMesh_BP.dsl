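; Editor tool Blueprint that converts a static mesh into a skinned skeletal mesh and generates temporary Athena cosmetic character assets for it.
; Geometry Script copies the template's bones, transfers weights by closest surface point, optionally prunes bones below the hands, and saves the new mesh.
; It can duplicate the template Skeleton and an AnimBP.
; A separate function duplicates CID and body/head part templates by gender and size, then links them through embedded Python.
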(fn getBonesToPrune
  (bind _dyn_mesh (LocalVariables|Getdyn_mesh))
  (LocalVariables|Setbones_to_prune (Utilities|Array|MakeArray "None"))
  (Utilities|FlowControl|Sequence
    (:then_0
      ;[0.7552 0.3107 0.678 0.5] Get theTemplate SkeletalMesh
      (bind _returnvalue (Game|ConstructObjectfromClass "/Script/GeometryFramework.DynamicMesh"))
      (bind _returnvalue_2 (GeometryScript|SkeletalMesh|CopyMeshfromSkeletalMesh (Variables|Default|GetTemplate) _returnvalue)
        (:Success
          (LocalVariables|Setdyn_mesh _returnvalue_2)
          (for _array_element (LocalVariables|Gethand_bone_names)
            (bind _bisvalidbonename (GeometryScript|MeshQueries|BoneWeights|GetBoneInfo _dyn_mesh (Utilities|String|StringToName _array_element)))
            (if _bisvalidbonename
              (bind (_bisvalidbonename_2 _childreninfo) (GeometryScript|MeshQueries|BoneWeights|GetBoneChildren _dyn_mesh (Utilities|String|StringToName _array_element) true))
              (for _array_element_2 _childreninfo
                ;[1 1 1] Don't include specific bones
                (if (not (Utilities|Array|ContainsItem (LocalVariables|Gethand_bone_names) (Utilities|String|ToString(Name) (.Name (Utilities|Struct|BreakBoneInfo _array_element_2)))))
                  (Utilities|Array|AddUnique (LocalVariables|Getbones_to_prune) (.Name (Utilities|Struct|BreakBoneInfo _array_element_2))))))))))
    (:then_1
      (return (LocalVariables|Getbones_to_prune)))))

(fn processDataAssets
  (params (AssetName :type String))
  (bind _assetname (Variables|GetAssetName))
  (bind _head_gender_path (LocalVariables|Gethead_gender_path))
  (bind _returnvalue (Utilities|String|Append _head_gender_path "CP_Athena_Head_" _assetname "_TEMP"))
  (bind _assetname_2 (Variables|GetAssetName))
  (bind _returnvalue_2 (Utilities|String|Append "/BRCosmetics/Athena/Items/Cosmetics/Characters/" "Character_" _assetname_2 "_TEMP"))
  (bind _assetname_3 (Variables|GetAssetName))
  (bind _returnvalue_3 (Utilities|String|Append "/BRCosmetics/Athena/Heroes/Meshes/Bodies/" "CP_Athena_Body_" _assetname_3 "_TEMP"))
  (Utilities|FlowControl|Sequence
    (:then_0
      ;[0.5469 0.3241 0.2507 0.5] Get Template Prefix
      (bind _returnvalue_4 (EditorScripting|Asset|GetPathNameforLoadedAsset (Variables|Default|GetTemplate) :!class /Script/EditorScriptingUtilities.EditorAssetLibrary))
      (LocalVariables|Settemplate_prefix (Utilities|String|Split (.FilenamePart (Utilities|Paths|Split _returnvalue_4)) "_Template")))
    (:then_1
      ;[0.15 0.06028 0.1377 0.5] Get Gender & Size
      (Utilities|FlowControl|Sequence
        (:then_0
          (if (Utilities|String|Contains (LocalVariables|Gettemplate_prefix) "_LRG_" true)
            (LocalVariables|Setis_lrg true)
            (LocalVariables|Setcid_size_str "_LRG")
            (LocalVariables|Setcp_size_str "LRG")
            (else
              (LocalVariables|Setis_lrg false)
              (LocalVariables|Setcid_size_str "_MED")
              (LocalVariables|Setcp_size_str ""))))
        (:then_1
          (if (Utilities|String|StartsWith (LocalVariables|Gettemplate_prefix) "M_")
            (LocalVariables|Setis_male true)
            (LocalVariables|Setgender_str "M")
            (LocalVariables|Sethead_gender_path "/BRCosmetics/Characters/CharacterParts/Male/Medium/Heads/")
            (else
              (LocalVariables|Setis_male false)
              (LocalVariables|Setgender_str "F")
              (LocalVariables|Sethead_gender_path "/BRCosmetics/Characters/CharacterParts/Female/Medium/Heads/"))))))
    (:then_2
      ;[0.9427 0.4959 0.9262 0.5] Create TEMP CID
      (bind _returnvalue_5 (EditorScripting|Asset|LoadAsset (Utilities|String|Append "/BRCosmetics/Athena/Items/Cosmetics/Characters/" "Character_CopyTemplate_" (LocalVariables|Getgender_str) (LocalVariables|Getcid_size_str)) :!class /Script/EditorScriptingUtilities.EditorAssetLibrary))
      (Utilities|IsValid _returnvalue_5
        (:Is%20Valid
          (bind _returnvalue_6 (EditorScripting|Asset|DoesAssetExist _returnvalue_2 :!class /Script/EditorScriptingUtilities.EditorAssetLibrary))
          (if _returnvalue_6
            (bind _returnvalue_7 (EditorScripting|Asset|LoadAsset _returnvalue_2 :!class /Script/EditorScriptingUtilities.EditorAssetLibrary))
            (bind _asathena_character_item_definition (Utilities|Casting|CastToAthenaCharacterItemDefinition _returnvalue_7))
            (LocalVariables|Setcid_asset _asathena_character_item_definition)
            (else
              (bind _returnvalue_8 (EditorScripting|Asset|DuplicateLoadedAsset _returnvalue_5 _returnvalue_2 :!class /Script/EditorScriptingUtilities.EditorAssetLibrary))
              (bind _asathena_character_item_definition_2 (Utilities|Casting|CastToAthenaCharacterItemDefinition _returnvalue_8))
              (LocalVariables|Setcid_asset _asathena_character_item_definition_2))))))
    (:then_3
      (Utilities|FlowControl|Sequence
        (:then_0
          ;[0.1367 0.2904 0.5469 0.5] Create TEMP CPs
          (bind _returnvalue_9 (EditorScripting|Asset|LoadAsset (Utilities|String|Append (LocalVariables|Gettemplate_path) "CP_Athena_Body_" (LocalVariables|Getgender_str) "_CopyTemplate" (LocalVariables|Getcp_size_str)) :!class /Script/EditorScriptingUtilities.EditorAssetLibrary))
          (Utilities|IsValid _returnvalue_9
            (:Is%20Valid
              (bind _returnvalue_10 (EditorScripting|Asset|DoesAssetExist _returnvalue_3 :!class /Script/EditorScriptingUtilities.EditorAssetLibrary))
              (if _returnvalue_10
                (bind _returnvalue_11 (EditorScripting|Asset|LoadAsset _returnvalue_3 :!class /Script/EditorScriptingUtilities.EditorAssetLibrary))
                (bind _ascustom_character_part (Utilities|Casting|CastToCustomCharacterPart _returnvalue_11))
                (LocalVariables|Setcp_body_asset _ascustom_character_part)
                (else
                  (bind _returnvalue_12 (EditorScripting|Asset|DuplicateLoadedAsset _returnvalue_9 _returnvalue_3 :!class /Script/EditorScriptingUtilities.EditorAssetLibrary))
                  (bind _ascustom_character_part_2 (Utilities|Casting|CastToCustomCharacterPart _returnvalue_12))
                  (LocalVariables|Setcp_body_asset _ascustom_character_part_2))))))
        (:then_1
          (bind _returnvalue_13 (EditorScripting|Asset|LoadAsset (Utilities|String|Append (LocalVariables|Gettemplate_path) "CP_Athena_Head_" (LocalVariables|Getgender_str) "_CopyTemplate" (LocalVariables|Getcp_size_str)) :!class /Script/EditorScriptingUtilities.EditorAssetLibrary))
          (Utilities|IsValid _returnvalue_13
            (:Is%20Valid
              (bind _returnvalue_14 (EditorScripting|Asset|DoesAssetExist _returnvalue :!class /Script/EditorScriptingUtilities.EditorAssetLibrary))
              (if _returnvalue_14
                (bind _returnvalue_15 (EditorScripting|Asset|LoadAsset _returnvalue :!class /Script/EditorScriptingUtilities.EditorAssetLibrary))
                (bind _ascustom_character_part_3 (Utilities|Casting|CastToCustomCharacterPart _returnvalue_15))
                (LocalVariables|Setcp_head_asset _ascustom_character_part_3)
                (else
                  (bind _returnvalue_16 (EditorScripting|Asset|DuplicateLoadedAsset _returnvalue_13 _returnvalue :!class /Script/EditorScriptingUtilities.EditorAssetLibrary))
                  (bind _ascustom_character_part_4 (Utilities|Casting|CastToCustomCharacterPart _returnvalue_16))
                  (LocalVariables|Setcp_head_asset _ascustom_character_part_4))))))
        (:then_2
          ;[0.599 0.3862 0.2341 0.5] Associate all of the files
          (bind _skeletalmesh (Variables|Default|GetSkeletalMesh)
            (:then
              (Utilities|FlowControl|Sequence
                (:then_0
                  (bind _returnvalue_17 (EditorScripting|Asset|GetPathNameforLoadedAsset _skeletalmesh :!class /Script/EditorScriptingUtilities.EditorAssetLibrary))
                  (bind _returnvalue_18 (EditorScripting|Asset|LoadAsset (Utilities|String|Append (Utilities|Paths|Split _returnvalue_17) "/" (.FilenamePart (Utilities|Paths|Split _returnvalue_17)) "_AnimBP") :!class /Script/EditorScriptingUtilities.EditorAssetLibrary))
                  (Utilities|IsValid _returnvalue_18
                    (:Is%20Valid
                      (LocalVariables|Setanim_bp_asset _returnvalue_18)
                      (Python|Execution|ExecutePythonScript "import unreal

asset_registry = unreal.AssetRegistryHelpers.get_asset_registry()

def main(cid, skeletalmesh, body_cp, head_cp, anim_bp, debug=True):	

	print(cid)
	cid_name = cid.get_name()
	if debug: print(f\"CID Asset: {cid_name}\")
	if debug: print(f\"SkeletalMesh: {skeletalmesh}\")		
	
	body_cp.set_editor_property(\"SkeletalMesh\", skeletalmesh)		
	anim_class = anim_bp.generated_class()		
	body_cp.get_editor_property(\"AdditionalData\").set_editor_property(\"AnimClass\", anim_class)		
	
	# update AnimBP skeletalMesh preview
	print(f\"AnimBP: {anim_bp}\")
	anim_bp.set_preview_mesh(skeletalmesh)		
	
	# new object array of character parts
	new_character_parts = [body_cp, head_cp]
	cid.set_editor_property(\"BaseCharacterParts\", new_character_parts)
	print(f\"Updated CID character parts\")
	
	# query updated character parts
	character_parts = cid.get_editor_property(\"BaseCharacterParts\")
	if character_parts:
		print(\"CharacterParts Updated:\")
		for i in character_parts:
			print(i)		


#skeletalmesh_path = \"/Game/00_TEST/SK_Ozone\"
#anim_bp_path = \"/Game/00_TEST/SK_Ozone_AnimBP\"
#cid_asset_path = \"/BRCosmetics/Athena/Items/Cosmetics/Characters/Character_Ozone_TEMP\"
#body_cp_path = \"/BRCosmetics/Athena/Heroes/Meshes/Bodies/CP_Athena_Body_Ozone_TEMP\"
#head_cp_path = \"/BRCosmetics/Characters/CharacterParts/Male/Medium/Heads/CP_Athena_Head_Ozone_TEMP\"
#skeletalmesh = unreal.load_asset(skeletalmesh_path)
#body_cp = unreal.load_asset(body_cp_path)
#head_cp = unreal.load_asset(head_cp_path)
#anim_bp = unreal.load_asset(anim_bp_path)
#cid = unreal.load_asset(cid_asset_path)
character_part = main(cid, skeletalmesh, body_cp, head_cp, anim_bp)" (LocalVariables|Getcid_asset) (Variables|Default|GetSkeletalMesh) (LocalVariables|Getcp_body_asset) (LocalVariables|Getcp_head_asset) (LocalVariables|Getanim_bp_asset)))))
                (:then_1
                  (EditorScripting|Asset|SaveLoadedAsset (Variables|Default|GetSkeletalMesh) :!class /Script/EditorScriptingUtilities.EditorAssetLibrary)
                  (EditorScripting|Asset|SaveLoadedAsset (LocalVariables|Getcid_asset) :!class /Script/EditorScriptingUtilities.EditorAssetLibrary)
                  (EditorScripting|Asset|SaveLoadedAsset (LocalVariables|Getcp_body_asset) :!class /Script/EditorScriptingUtilities.EditorAssetLibrary)
                  (EditorScripting|Asset|SaveLoadedAsset (LocalVariables|Getcp_head_asset) :!class /Script/EditorScriptingUtilities.EditorAssetLibrary)
                  (EditorScripting|Asset|SaveLoadedAsset (LocalVariables|Getanim_bp_asset) :!class /Script/EditorScriptingUtilities.EditorAssetLibrary))))
            (:else)))))
    (:then_4
      (return (LocalVariables|Getcid_asset)))))

(fn processStaticToSkeletal
  (params (AssetName :type String) (CreateAnimBP :type Boolean :default true) (CreateSkeleton :type Boolean :default true))
  (bind _valid (CallFunction|ValidateAssets))
  (if _valid
    (Utilities|FlowControl|Sequence
      (:then_0
        ;[0.15 0.06863 0.1496 0.5] Create Dynamic Meshes
        (bind _returnvalue (Game|ConstructObjectfromClass "/Script/GeometryFramework.DynamicMesh"))
        (bind _returnvalue_2 (GeometryScript|SkeletalMesh|CopyMeshfromSkeletalMesh (Variables|Default|GetTemplate) _returnvalue)
          (:Success
            (Variables|Default|Setdyn_template _returnvalue_2)
            (bind _returnvalue_3 (Game|ConstructObjectfromClass "/Script/GeometryFramework.DynamicMesh"))
            (bind _returnvalue_4 (GeometryScript|StaticMesh|CopyMeshFromStaticMesh (Variables|Default|GetStaticMesh) _returnvalue_3)
              (:Success
                (bind _output_get (Variables|Default|Setdyn_static _returnvalue_4))
                (GeometryScript|MeshQueries|BoneWeights|CopyBonesfromSkeleton (Variables|/Script/Engine.SkeletalMesh|GetSkeleton (Variables|Default|GetTemplate)) _output_get (Utilities|Struct|MakeGeometryScriptCopyBonesfromMeshOptions false)))))))
      (:then_1
        ;[0.1367 0.2904 0.5469 0.5] Transfer and Update Skinning
        (bind _returnvalue_5 (GeometryScript|MeshQueries|BoneWeights|TransferBoneWeightsfromMesh (Variables|Default|Getdyn_template) (Variables|Default|Getdyn_static) (Utilities|Struct|MakeGeometryScriptTransferBoneWeightsOptions "ClosestPointOnSurface" "TargetBones")))
        (if (Variables|Default|Getprune_fingers_skinning)
          (bind _extra_bones (CallFunction|GetBonestoPrune))
          (GeometryScript|MeshQueries|BoneWeights|PruneBoneWeights _returnvalue_5 _extra_bones (Utilities|Struct|MakeGeometryScriptPruneBoneWeightsOptions "ReassignToParent") "(ProfileName=\"Default\")")))
      (:then_2
        ;[0.9427 0.4959 0.9262 0.5] Save new SkeletalMesh
        (bind (_materiallist _materialslotnames) (GeometryScript|StaticMesh|GetMaterialListfromStaticMesh (Variables|Default|GetStaticMesh)))
        (bind _returnvalue_6 (GeometryScript|Materials|ConvertMaterialListtoMaterialMap _materiallist _materialslotnames))
        (Utilities|FlowControl|Sequence
          (:then_0
            (bind _returnvalue_7 (EditorScripting|Asset|GetPathNameforLoadedAsset (Variables|Default|GetStaticMesh) :!class /Script/EditorScriptingUtilities.EditorAssetLibrary))
            (bind _returnvalue_8 (GeometryScript|AssetManagement|CreateNewSkeletalMeshAssetfromMesh (Variables|Default|Getdyn_static) (Variables|/Script/Engine.SkeletalMesh|GetSkeleton (Variables|Default|GetTemplate)) (Utilities|String|Append (Utilities|Paths|Split _returnvalue_7) "/" (Variables|GetAssetName)) (Utilities|Struct|MakeGeometryScriptCreateNewSkeletalMeshAssetOptions false false _returnvalue_6 false false "(bEnabled=False,bExplicitTangents=False,bLerpUVs=True,bSeparable=False,bVoxelNDF=True,bVoxelOpacity=False,ShapePreservation=None,PositionPrecision=-2147483648,NormalPrecision=-1,TangentPrecision=-1,BoneWeightPrecision=-1,TargetMinimumResidencyInKB=0,KeepPercentTriangles=1.000000,TrimRelativeError=0.000000,GenerateFallback=PlatformDefault,FallbackTarget=Auto,FallbackPercentTriangles=1.000000,FallbackRelativeError=1.000000,MaxEdgeLengthFactor=0.000000,NumRays=64,VoxelLevel=0,RayBackUp=0.000000,DisplacementUVChannel=0,DisplacementMaps=,NaniteAssemblyData=(Parts=,Nodes=))" true))
              (:Success
                (Variables|Default|SetSkeletalMesh _returnvalue_8))))
          (:then_1
            (bind _skeletalmesh (Variables|Default|GetSkeletalMesh)
              (:then
                (Utilities|FlowControl|Sequence
                  (:then_0
                    (if (Variables|GetCreateSkeleton)
                      ;[0.7135 0.2867 0.2321 0.5] Create the Skeleton
                      (bind _returnvalue_9 (EditorScripting|Asset|GetPathNameforLoadedAsset _skeletalmesh :!class /Script/EditorScriptingUtilities.EditorAssetLibrary))
                      (bind _returnvalue_10 (EditorScripting|Asset|DoesAssetExist (Utilities|String|Append (Utilities|Paths|Split _returnvalue_9) "/" (.FilenamePart (Utilities|Paths|Split _returnvalue_9)) "_Skeleton") :!class /Script/EditorScriptingUtilities.EditorAssetLibrary))
                      (if _returnvalue_10
                        (bind _returnvalue_11 (EditorScripting|Asset|LoadAsset (Utilities|String|Append (Utilities|Paths|Split _returnvalue_9) "/" (.FilenamePart (Utilities|Paths|Split _returnvalue_9)) "_Skeleton") :!class /Script/EditorScriptingUtilities.EditorAssetLibrary))
                        (bind _asskeleton (Utilities|Casting|CastToSkeleton _returnvalue_11))
                        (Variables|/Script/Engine.SkeletalMesh|SetSkeleton _asskeleton (Variables|Default|GetSkeletalMesh))
                        (else
                          (bind _returnvalue_12 (EditorScripting|Asset|DuplicateLoadedAsset (Variables|/Script/Engine.SkeletalMesh|GetSkeleton (Variables|Default|GetTemplate)) (Utilities|String|Append (Utilities|Paths|Split _returnvalue_9) "/" (.FilenamePart (Utilities|Paths|Split _returnvalue_9)) "_Skeleton") :!class /Script/EditorScriptingUtilities.EditorAssetLibrary))
                          (bind _asskeleton_2 (Utilities|Casting|CastToSkeleton _returnvalue_12))
                          (bind _output_get_2 (Variables|/Script/Engine.SkeletalMesh|SetSkeleton _asskeleton_2 (Variables|Default|GetSkeletalMesh)))
                          (EditorScripting|Asset|SaveLoadedAsset _output_get_2 :!class /Script/EditorScriptingUtilities.EditorAssetLibrary)
                          (EditorScripting|Asset|SaveLoadedAsset _asskeleton_2 :!class /Script/EditorScriptingUtilities.EditorAssetLibrary)))))
                  (:then_1
                    (if (Variables|GetCreateAnimBP)
                      ;[0.15 0.06028 0.1377 0.5] Create the AnimBP
                      (Utilities|FlowControl|Sequence
                        (:then_0
                          (bind _returnvalue_13 (EditorScripting|Asset|GetPathNameforLoadedAsset (Variables|Default|GetSkeletalMesh) :!class /Script/EditorScriptingUtilities.EditorAssetLibrary))
                          (LocalVariables|Setanimbp_path (Utilities|String|Append (Utilities|Paths|Split _returnvalue_13) "/" (.FilenamePart (Utilities|Paths|Split _returnvalue_13)) "_AnimBP")))
                        (:then_1
                          (bind _returnvalue_14 (EditorScripting|Asset|DoesAssetExist (LocalVariables|Getanimbp_path) :!class /Script/EditorScriptingUtilities.EditorAssetLibrary))
                          (if _returnvalue_14
                            (EditorScripting|Asset|LoadAsset (LocalVariables|Getanimbp_path) :!class /Script/EditorScriptingUtilities.EditorAssetLibrary)
                            (else
                              (bind _returnvalue_15 (EditorScripting|Asset|LoadAsset (LocalVariables|Getanimbp_template_path) :!class /Script/EditorScriptingUtilities.EditorAssetLibrary))
                              (Utilities|IsValid _returnvalue_15
                                (:Is%20Valid
                                  (bind _returnvalue_16 (EditorScripting|Asset|DuplicateLoadedAsset _returnvalue_15 (LocalVariables|Getanimbp_path) :!class /Script/EditorScriptingUtilities.EditorAssetLibrary))
                                  (bind _output_get_3 (LocalVariables|Setanimbp _returnvalue_16))
                                  (EditorScripting|Asset|SaveLoadedAsset _output_get_3 :!class /Script/EditorScriptingUtilities.EditorAssetLibrary)))))))))))
              (:else)))))
      (:then_3
        (return (Variables|Default|GetSkeletalMesh))))))

(fn validateAssets
  (Variables|Default|GetStaticMesh
    (:then
      (Variables|Default|GetTemplate
        (:then
          (return true))
        (:else
          (return false))))
    (:else
      (return false))))
